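# Sound cue for the assault rifle pickup sound, classed as loot collection audio.
# A quality-level root sends its first two levels and last two levels to separate mixers, each boosting its own Rifle_Pickup wave asset and leaving a second input empty.
# The cue lowers volume and pitch, uses weapon-pickup attenuation, and references a shared PickupMaximum_Oldest concurrency set.
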SoundCue /Game/Sounds/Fort_GamePlay_Sounds/Loot/AR_Pickup_Cue
root=QualityLevel_0
nodes [
 {id=Mixer_0, children [WavePlayer_6,null], properties {inputVolume [1.2,1]}}
 {id=Mixer_1, children [WavePlayer_1,null], properties {inputVolume [1.2,1]}}
 {id=QualityLevel_0, children [Mixer_1,Mixer_1,Mixer_0,Mixer_0]}
 {id=WavePlayer_1, properties {soundWaveAssetPtr=&/Rifle_Pickup}}
 {id=WavePlayer_6, properties {soundWaveAssetPtr=&Rifle_Pickup}}
]
properties {
 attenuationSettings=&Fort_Weapon_Pickup_Att
 maxDistance=3500
 pitchMultiplier=0.9
 soundClassObject=&Fort_Loot_Collect
 volumeMultiplier=0.6
 concurrencyOverrides {maxCount=1, resolutionRule=PreventNew}
 concurrencySet [&PickupMaximum_Oldest]
}
