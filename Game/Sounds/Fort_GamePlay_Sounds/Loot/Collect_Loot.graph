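# Sound cue for collecting loot, layering a loot pickup sound with a delayed backpack stow sound.
# A mixer plays the collect wave immediately alongside the backpack-in wave, which a fixed delay holds back by just under a second.
# Playback uses prop search attenuation and a loot reward sound class, with a concurrency override that allows one instance at a time and prevents new ones.
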SoundCue /Game/Sounds/Fort_GamePlay_Sounds/Loot/Collect_Loot
root=Mixer_0
nodes [
 {id=Delay_0, children [WavePlayer_0], properties {delayMax=0.95, delayMin=0.95}}
 {id=Mixer_0, children [WavePlayer_2,Delay_0], properties {inputVolume [1,1]}}
 {id=WavePlayer_0, properties {soundWaveAssetPtr=&/Collect_Loot_BackpackIn03}}
 {id=WavePlayer_2, properties {soundWaveAssetPtr=&/Collect_Loot01}}
]
properties {
 attenuationSettings=/Game/Sounds/Attenuation/Fort_Prop_Search_Att
 bOverrideConcurrency=true
 maxDistance=1300
 soundClassObject=&Fort_Loot_Reward_PSM
 volumeMultiplier=0.7
 concurrencyOverrides {maxCount=1, resolutionRule=PreventNew}
}
