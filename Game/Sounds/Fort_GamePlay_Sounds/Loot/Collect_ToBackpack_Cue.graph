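# Sound cue for loot collection into the backpack, routed to the Fort_Loot_Collect sound class.
# A Delay node feeds a single wave player for Collect_ToBackpack07, with the cue pitched up and heavily reduced in volume.
# It uses prop-search attenuation and a concurrency override that allows one instance at a time, blocking new plays rather than cutting off the active one.
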SoundCue /Game/Sounds/Fort_GamePlay_Sounds/Loot/Collect_ToBackpack_Cue
root=Delay_1
nodes [
 {id=Delay_1, children [WavePlayer_1]}
 {id=WavePlayer_1, properties {soundWaveAssetPtr=&/Collect_ToBackpack07}}
]
properties {
 attenuationSettings=/Game/Sounds/Attenuation/Fort_Prop_Search_Att
 bOverrideConcurrency=true
 maxDistance=1300
 pitchMultiplier=2
 soundClassObject=&Fort_Loot_Collect
 volumeMultiplier=0.1
 concurrencyOverrides {maxCount=1, resolutionRule=PreventNew}
}
