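# Generic sound cue for dropped loot in Fortnite gameplay.
# A single wave player feeds the root directly, playing the UI_DropItem02 wave at reduced volume with no randomization or modulation.
# It uses weapon-drop attenuation and the loot-collect sound class.
# A concurrency override prevents new instances from starting once the limit is reached, rather than stopping playing ones.
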SoundCue /Game/Sounds/Fort_GamePlay_Sounds/Loot/Drop_Loot_Generic_Cue
root=WavePlayer_0
nodes [{id=WavePlayer_0, properties {soundWaveAssetPtr=/Game/Sounds/Fort_UI_Sounds/Gameplay/UI_DropItem02}}]
properties {
 attenuationSettings=&Fort_Weapon_Drop_Att
 bOverrideConcurrency=true
 maxDistance=3000
 soundClassObject=&Fort_Loot_Collect
 volumeMultiplier=0.5
 concurrencyOverrides {resolutionRule=PreventNew}
}
