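# Sound cue for the health-restore sound of eating a consumable in Athena, judging by its name.
# A single health-pickup wave feeds a Modulator for slight random variation, with the cue pitched up and at reduced volume.
# It plays in the consumables sound class with a shared consumable-use attenuation asset.
# Its own concurrency override allows only a couple of voices, stopping the oldest with a short release.
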SoundCue /Game/Sounds/Fort_GamePlay_Sounds/Loot/EatHealth_Restore_Athena_Cue
root=Modulator_0
nodes [
 {id=Modulator_0, children [WavePlayer_31], properties {volumeMax=1}}
 {id=WavePlayer_31, properties {soundWaveAssetPtr=/Game/Sounds/Fort_Gadgets/Unassigned/new_health_get}}
]
properties {
 attenuationSettings=&Fort_Consumable_Use_Athena_Att
 bOverrideConcurrency=true
 maxDistance=2250
 pitchMultiplier=1.5
 soundClassObject=&Fort_Consumables
 volumeMultiplier=0.5
 attenuationOverrides {falloffDistance=1200, attenuationShapeExtents {x=800}}
 concurrencyOverrides {maxCount=2, resolutionRule=StopOldest, voiceStealReleaseTime=0.2}
}
