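# Sound cue for a health-restore eating sound, filed under gameplay loot sounds.
# A single wave player feeds the root directly with the new_health_get wave, with no randomization or mixing.
# It routes to the consumables sound class with boosted volume and uses the shared weapon-drop attenuation asset for spatial falloff.
# A concurrency override prevents new instances from starting once the limit is reached, rather than cutting off playing ones.
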SoundCue /Game/Sounds/Fort_GamePlay_Sounds/Loot/EatHealth_Restore_Cue
root=WavePlayer_31
nodes [{id=WavePlayer_31, properties {soundWaveAssetPtr=/Game/Sounds/Fort_Gadgets/Unassigned/new_health_get}}]
properties {
 attenuationSettings=&Fort_Weapon_Drop_Att
 bOverrideConcurrency=true
 maxDistance=3000
 soundClassObject=&Fort_Consumables
 volumeMultiplier=1.3
 attenuationOverrides {falloffDistance=1200, attenuationShapeExtents {x=800}}
 concurrencyOverrides {resolutionRule=PreventNew}
}
