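# Sound cue for a guitar being dropped, classed as loot collect audio with weapon-drop attenuation.
# A mixer layers a generic weapon-drop sound with a guitar impact sample that follows after a short fixed delay.
# The cue plays at reduced volume and overrides concurrency to block new instances rather than cut off ones already playing.
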SoundCue /Game/Sounds/Fort_GamePlay_Sounds/Loot/Guitar_Drop_Cue
root=Mixer_0
nodes [
 {id=Delay_9, children [WavePlayer_129], properties {delayMax=0.1, delayMin=0.1}}
 {id=Mixer_0, children [WavePlayer_0,Delay_9], properties {inputVolume [1,1]}}
 {id=WavePlayer_0, properties {soundWaveAssetPtr=&/fort_weapon_drop_01_mix}}
 {id=WavePlayer_129, properties {soundWaveAssetPtr=/Game/Sounds/Fort_Weapons/Guitar/guitar_impact_04}}
]
properties {
 attenuationSettings=&Fort_Weapon_Drop_Att
 bOverrideConcurrency=true
 maxDistance=3000
 soundClassObject=&Fort_Loot_Collect
 volumeMultiplier=0.5
 concurrencyOverrides {resolutionRule=PreventNew}
}
