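# Sound cue for ammo pickup, using the loot-collect sound class and weapon pickup attenuation.
# A mixer layers an immediate deep loot sound with a delayed ammo collect sound and a slightly later backpack-in sound.
# A quality-level switch sends the first two levels to that mix and the last two to a single optimized wave.
# The PickupMaximum_Oldest concurrency set governs overlapping plays.
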SoundCue /Game/Sounds/Fort_GamePlay_Sounds/Loot/Pickup_Ammo_Cue
root=QualityLevel_0
nodes [
 {id=Delay_3, children [WavePlayer_28], properties {delayMax=0.95, delayMin=0.95}}
 {id=Delay_4, children [WavePlayer_19], properties {delayMax=0.75, delayMin=0.75}}
 {id=Mixer_2, children [Delay_4, Delay_3, WavePlayer_38], properties {inputVolume [1.3,1,0.75]}}
 {id=QualityLevel_0, children [Mixer_2,Mixer_2,WavePlayer_0,WavePlayer_0]}
 {id=WavePlayer_0, properties {soundWaveAssetPtr=&/Pickup_Ammo_opt}}
 {id=WavePlayer_19, properties {soundWaveAssetPtr=&/Loot_AmmoCollect01}}
 {id=WavePlayer_28, properties {soundWaveAssetPtr=&/Collect_Loot_BackpackIn01}}
 {id=WavePlayer_38, properties {soundWaveAssetPtr=&/Collect_LootDeep01}}
]
properties {
 attenuationSettings=&Fort_Weapon_Pickup_Att
 maxDistance=3500
 soundClassObject=&Fort_Loot_Collect
 volumeMultiplier=0.159
 concurrencyOverrides {maxCount=1, resolutionRule=PreventNew}
 concurrencySet [&PickupMaximum_Oldest]
}
