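# Sound cue for a rare weapon pickup flourish, using weapon pickup attenuation, the loot-collect sound class and a pickup concurrency limit.
# An IsLocal branch plays sound only when true, leaving the false and unset paths silent.
# It feeds a quality-level switch whose first two slots share one volume-varying modulator, with the other slots empty.
# The modulator draws equally from four flourish waves through a random node excluded from branch culling.
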SoundCue /Game/Sounds/Fort_GamePlay_Sounds/Loot/Pickup_Weapon_Flourish_Cue
root=Branch_0
nodes [
 {id=Branch_0, children [QualityLevel_1, null, null], properties {boolParameterName=IsLocal}}
 {id=Modulator_0, children [Random_0], properties {pitchMax=1, pitchMin=1, volumeMax=1.3, volumeMin=1}}
 {id=QualityLevel_1, children [Modulator_0,Modulator_0,null,null]}
 {
  id=Random_0
  children [WavePlayer_1,WavePlayer_2,WavePlayer_3,WavePlayer_4]
  properties {bShouldExcludeFromBranchCulling=true, weights [1,1,1,1]}
 }
 {id=WavePlayer_1, properties {soundWaveAssetPtr=&/Pickup_Weapon_Flourish_Rare_01}}
 {id=WavePlayer_2, properties {soundWaveAssetPtr=&/Pickup_Weapon_Flourish_Rare_02}}
 {id=WavePlayer_3, properties {soundWaveAssetPtr=&/Pickup_Weapon_Flourish_Rare_03}}
 {id=WavePlayer_4, properties {soundWaveAssetPtr=&/Pickup_Weapon_Flourish_Rare_04}}
]
properties {
 attenuationSettings=&Fort_Weapon_Pickup_Att
 maxDistance=3500
 soundClassObject=&Fort_Loot_Collect
 volumeMultiplier=0.301
 concurrencyOverrides {maxCount=1, resolutionRule=PreventNew}
 concurrencySet [&PickupMaximum_Oldest]
}
