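# Sound cue for picking up a pistol as loot in Athena, in the loot-collect sound class.
# A Mixer layers four waves at weighted volumes: a weapon reward and a deep collect sound play at once.
# A louder pistol pickup and a backpack-in sound follow just under a second later through fixed Delays.
# Overall volume is reduced, with weapon-pickup attenuation and a shared pickup concurrency set.
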SoundCue /Game/Sounds/Fort_GamePlay_Sounds/Loot/Pistol_Pickup_Athena_1_8_Cue
root=Mixer_2
nodes [
 {id=Delay_2, children [WavePlayer_4], properties {delayMax=0.85, delayMin=0.85}}
 {id=Delay_3, children [WavePlayer_5], properties {delayMax=0.95, delayMin=0.95}}
 {id=Mixer_2, children [WavePlayer_7,Delay_2,Delay_3,WavePlayer_6], properties {inputVolume [1,2,1,0.75]}}
 {id=WavePlayer_4, properties {soundWaveAssetPtr=&/Pistol_Pickup}}
 {id=WavePlayer_5, properties {soundWaveAssetPtr=&/Collect_Loot_BackpackIn01}}
 {id=WavePlayer_6, properties {soundWaveAssetPtr=&/Collect_LootDeep01}}
 {id=WavePlayer_7, properties {soundWaveAssetPtr=&/new_weapon_loot_reward_02}}
]
properties {
 attenuationSettings=&Fort_Weapon_Pickup_Att
 maxDistance=3500
 soundClassObject=&Fort_Loot_Collect
 volumeMultiplier=0.4
 concurrencyOverrides {maxCount=1, resolutionRule=PreventNew}
 concurrencySet [&PickupMaximum_Oldest]
}
