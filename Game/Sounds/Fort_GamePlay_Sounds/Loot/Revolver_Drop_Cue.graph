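# Sound cue for dropping a revolver as loot, playing a pistol drop sound.
# A quality-level root sends its first two levels to a mixer layering the drop with a backpack loot-collect sound.
# Its last two levels play only the drop wave.
# It uses weapon-drop attenuation, the loot-collect sound class and a stop-oldest weapon-drop concurrency set.
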SoundCue /Game/Sounds/Fort_GamePlay_Sounds/Loot/Revolver_Drop_Cue
root=QualityLevel_0
nodes [
 {id=Mixer_7, children [WavePlayer_8, WavePlayer_30], properties {inputVolume [1,1]}}
 {id=QualityLevel_0, children [Mixer_7,Mixer_7,WavePlayer_0,WavePlayer_0]}
 {id=WavePlayer_0, properties {soundWaveAssetPtr=&/pistol_drop_06}}
 {id=WavePlayer_30, properties {soundWaveAssetPtr=&/Collect_Loot_BackpackIn01}}
 {id=WavePlayer_8, properties {soundWaveAssetPtr=&pistol_drop_06}}
]
properties {
 attenuationSettings=&Fort_Weapon_Drop_Att
 maxDistance=3000
 soundClassObject=&Fort_Loot_Collect
 volumeMultiplier=1
 concurrencyOverrides {maxCount=1, resolutionRule=PreventNew}
 concurrencySet [&Weapon_Drop_StopOldest]
}
