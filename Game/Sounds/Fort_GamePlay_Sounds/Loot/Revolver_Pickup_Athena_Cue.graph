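# Sound cue for the Athena revolver pickup sound, classed as loot collect.
# A quality switch at the root sends its first two levels to one mixer and the last two to another, each playing its own Revolver_Pickup wave reference with a boosted input.
# Weapon pickup attenuation shapes falloff, the cue's overall volume is reduced, and a pickup concurrency set is assigned alongside single-instance, prevent-new overrides.
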SoundCue /Game/Sounds/Fort_GamePlay_Sounds/Loot/Revolver_Pickup_Athena_Cue
root=QualityLevel_1
nodes [
 {id=Mixer_0, children [WavePlayer_5,null], properties {inputVolume [2,1]}}
 {id=Mixer_1, children [WavePlayer_6,null], properties {inputVolume [2,1]}}
 {id=QualityLevel_1, children [Mixer_0,Mixer_0,Mixer_1,Mixer_1]}
 {id=WavePlayer_5, properties {soundWaveAssetPtr=&/Revolver_Pickup}}
 {id=WavePlayer_6, properties {soundWaveAssetPtr=&Revolver_Pickup}}
]
properties {
 attenuationSettings=&Fort_Weapon_Pickup_Att
 maxDistance=3500
 soundClassObject=&Fort_Loot_Collect
 volumeMultiplier=0.398
 concurrencyOverrides {maxCount=1, resolutionRule=PreventNew}
 concurrencySet [&PickupMaximum_Oldest]
}
