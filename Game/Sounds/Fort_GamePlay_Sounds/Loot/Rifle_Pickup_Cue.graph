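# Sound cue for picking up a rifle as loot, layering a loot reward music stinger with a delayed rifle pickup sound.
# A quality-level root selects between two mixers, with the last two levels using an "opt" variant of the stinger.
# Each mixer plays the stinger immediately and brings in the slightly louder pickup wave after a short fixed delay.
# Playback uses weapon pickup attenuation, the loot-collect sound class and a shared pickup concurrency set.
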SoundCue /Game/Sounds/Fort_GamePlay_Sounds/Loot/Rifle_Pickup_Cue
root=QualityLevel_1
nodes [
 {id=Delay_0, children [WavePlayer_5], properties {delayMax=0.75, delayMin=0.75}}
 {id=Delay_1, children [WavePlayer_6], properties {delayMax=0.75, delayMin=0.75}}
 {id=Mixer_0, children [Delay_0,WavePlayer_0], properties {inputVolume [1.2,1]}}
 {id=Mixer_1, children [Delay_1,WavePlayer_1], properties {inputVolume [1.2,1]}}
 {id=QualityLevel_1, children [Mixer_0,Mixer_0,Mixer_1,Mixer_1]}
 {id=WavePlayer_0, properties {soundWaveAssetPtr=/Game/Sounds/Fort_Music/Stingers/loot_reward_02_mix}}
 {id=WavePlayer_1, properties {soundWaveAssetPtr=/Game/Sounds/Fort_Music/Stingers/loot_reward_02_mix_opt}}
 {id=WavePlayer_5, properties {soundWaveAssetPtr=&/Rifle_Pickup}}
 {id=WavePlayer_6, properties {soundWaveAssetPtr=&Rifle_Pickup}}
]
properties {
 attenuationSettings=&Fort_Weapon_Pickup_Att
 maxDistance=3500
 soundClassObject=&Fort_Loot_Collect
 volumeMultiplier=0.4
 concurrencyOverrides {maxCount=1, resolutionRule=PreventNew}
 concurrencySet [&PickupMaximum_Oldest]
}
