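# Sound cue for a shotgun drop, going by its name, in the gameplay loot sounds.
# A Mixer layers a weapon drop wave at full input volume with a slightly quieter deep loot-collect wave, with no randomization.
# The cue plays at reduced volume through weapon-drop attenuation and the loot-collect sound class, and references the Weapon_Drop_StopOldest concurrency set.
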SoundCue /Game/Sounds/Fort_GamePlay_Sounds/Loot/Shotgun_Drop_1_8_Cue
root=Mixer_0
nodes [
 {id=Mixer_0, children [WavePlayer_2,WavePlayer_0], properties {inputVolume [1,0.75]}}
 {id=WavePlayer_0, properties {soundWaveAssetPtr=&/Collect_LootDeep01}}
 {id=WavePlayer_2, properties {soundWaveAssetPtr=&/fort_weapons_drop_02}}
]
properties {
 attenuationSettings=&Fort_Weapon_Drop_Att
 maxDistance=3000
 soundClassObject=&Fort_Loot_Collect
 volumeMultiplier=0.5
 concurrencyOverrides {resolutionRule=PreventNew}
 concurrencySet [&Weapon_Drop_StopOldest]
}
