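# Sound cue for a sword weapon drop, classed as a loot collect sound.
# A mixer plays a generic weapon drop wave at reduced volume together with a sword slash wave that starts after a short fixed delay.
# The cue uses weapon drop attenuation and a quieter overall volume, and overrides concurrency to prevent new instances once the limit is reached.
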SoundCue /Game/Sounds/Fort_GamePlay_Sounds/Loot/Sword_Drop_Cue
root=Mixer_16
nodes [
 {id=Delay_10, children [WavePlayer_128], properties {delayMax=0.1, delayMin=0.1}}
 {id=Mixer_16, children [WavePlayer_0,Delay_10], properties {inputVolume [0.8,1]}}
 {id=WavePlayer_0, properties {soundWaveAssetPtr=&/fort_weapon_drop_01_mix}}
 {id=WavePlayer_128, properties {soundWaveAssetPtr=/Game/Sounds/Fort_Weapons/Swords/sword_slash_01}}
]
properties {
 attenuationSettings=&Fort_Weapon_Drop_Att
 bOverrideConcurrency=true
 maxDistance=3000
 soundClassObject=&Fort_Loot_Collect
 volumeMultiplier=0.4
 concurrencyOverrides {resolutionRule=PreventNew}
}
