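# Sound cue for the impact of dropped wood loot in gameplay.
# A random node picks one of four equally weighted impact waves, which passes through pitch and volume modulation and a short fixed delay.
# It uses weapon-drop attenuation and the loot-collect sound class at reduced volume.
# A concurrency override allows one instance at a time and blocks new plays rather than cutting off the current one.
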SoundCue /Game/Sounds/Fort_GamePlay_Sounds/Loot/Wood_Drop_Cue
root=Delay_0
nodes [
 {id=Delay_0, children [Modulator_0], properties {delayMax=0.3, delayMin=0.3}}
 {id=Modulator_0, children [Random_2], properties {pitchMax=1.1}}
 {
  id=Random_2
  children [WavePlayer_11, WavePlayer_12, WavePlayer_13, WavePlayer_14]
  properties {bShouldExcludeFromBranchCulling=true, weights [1,1,1,1]}
 }
 {id=WavePlayer_11, properties {soundWaveAssetPtr=&/Wood_Drop_Impact_01}}
 {id=WavePlayer_12, properties {soundWaveAssetPtr=&/Wood_Drop_Impact_02}}
 {id=WavePlayer_13, properties {soundWaveAssetPtr=&/Wood_Drop_Impact_03}}
 {id=WavePlayer_14, properties {soundWaveAssetPtr=&/Wood_Drop_Impact_04}}
]
properties {
 attenuationSettings=&Fort_Weapon_Drop_Att
 bOverrideConcurrency=true
 maxDistance=3000
 soundClassObject=&Fort_Loot_Collect
 volumeMultiplier=0.6
 concurrencyOverrides {maxCount=1, resolutionRule=PreventNew}
}
