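# Sound cue for the impact of wood loot landing, using the weapon drop attenuation and the loot collect sound class.
# A Random node picks one of four equally weighted impact waves and is excluded from branch culling.
# A root Modulator varies the pitch of each play.
# A concurrency override allows one instance at a time and blocks new plays while it is active.
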SoundCue /Game/Sounds/Fort_GamePlay_Sounds/Loot/Wood_Land_Cue
root=Modulator_0
nodes [
 {id=Modulator_0, children [Random_3], properties {pitchMax=1.2}}
 {
  id=Random_3
  children [WavePlayer_15, WavePlayer_16, WavePlayer_17, WavePlayer_18]
  properties {bShouldExcludeFromBranchCulling=true, weights [1,1,1,1]}
 }
 {id=WavePlayer_15, properties {soundWaveAssetPtr=&/Wood_Land_Impact_01}}
 {id=WavePlayer_16, properties {soundWaveAssetPtr=&/Wood_Land_Impact_02}}
 {id=WavePlayer_17, properties {soundWaveAssetPtr=&/Wood_Land_Impact_03}}
 {id=WavePlayer_18, properties {soundWaveAssetPtr=&/Wood_Land_Impact_04}}
]
properties {
 attenuationSettings=&Fort_Weapon_Drop_Att
 bOverrideConcurrency=true
 maxDistance=3000
 soundClassObject=&Fort_Loot_Collect
 volumeMultiplier=1
 concurrencyOverrides {maxCount=1, resolutionRule=PreventNew}
}
